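# Material function in the LandFoil player-body cosmetic folder that computes dithered triplanar UVs, by default from pre-skinned position and normal.
# Powered absolute normal weights, divided by R+R+B (green is absent), plus a temporal pixel-pattern and noise dither are floored into an axis mask.
# The mask selects one planar projection of scaled position, allowing single-sample triplanar texturing.
# It also outputs the mask and DDX/DDY from scaled mesh UVs.
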MaterialFunction /BRCosmetics/Characters/Player/Male/Medium/Bodies/M_MED_ZenCrash_LandFoil/Materials/MF_TriplanarDithered_PreSkinned
graph {nodes [
  n1 = VertexInterpolator(VS=LocalPosition() {localOrigin=InstancePreSkinning}.XYZ)
  n2 = FunctionInput() {bUsePreviewValueAsDefault=true, inputName=Texture Scale, inputType=FunctionInput_Scalar, previewValue=<120 120 120 120>}
  n3 = Divide(
   A=Multiply(
    A=FunctionInput(Preview=@n1.PS) {bUsePreviewValueAsDefault=true, inputName=InPosition, previewValue=<120 120 120 120>}
    B=Constant3Vector() {constant=<rgba 1 1 -1 0>}
   )
   B=@n2
  ) {constB=120}
  n4 = VertexInterpolator(VS=PreSkinnedNormal())
  n5 = FunctionInput() {bUsePreviewValueAsDefault=true, inputName=Triplanar Fade, inputType=FunctionInput_Scalar, sortPriority=1, previewValue=<10 10 10 1>}
  n6 = Power(
   Base=Abs(None=FunctionInput(Preview=@n4.PS) {bUsePreviewValueAsDefault=true, description=VertexNormalWS as Default, inputName=Normal})
   Exp=@n5
  ) {constExponent=10}
  n7 = MaterialFunctionCall(Float3=Saturate(None=@n6)) {materialFunction=&BreakOutFloat3Components}
  n8 = Divide(A=@n6, B=Add(A=Add(A=@n7.R, B=@n7.R), B=@n7.B))
  n9 = ViewProperty() {property=MEVP_TemporalSampleIndex}
  MaterialExpressionAdd_2 = Add(A=ScreenPosition().PixelPosition, B=AppendVector(A=@n9.Property, B=@n9.Property))
  n10 = Custom!(p=@MaterialExpressionAdd_2) {
   code="Mod( ((uint)(p.x) + 2 * (uint)(p.y)) , 5 )"
   outputType=CMOT_Float1
   inputs [{
     inputName=p
     input {expression=&self:MaterialExpressionAdd_2}
    }]
  }
  n11 = TextureSample(UVs=MaterialFunctionCall!(TextureResolution=Constant2Vector() {g=64, r=64}) {materialFunction=/Engine/Functions/Engine_MaterialFunctions02/Texturing/ScreenAlignedPixelToPixelUVs}.Result) {samplerType=SAMPLERTYPE_LinearGrayscale, texture=/Engine/EngineMaterials/Good64x64TilingNoiseHighFreq}
  n12 = Max(A=Multiply(A=Add(A=@n10, B=@n11.R) {constB=0}) {constB=0.1667}) {constB=0.001}
  n13 = OneMinus(None=@n12)
  n14 = Floor(None=Add(A=ComponentMask(None=@n8) {g=true, r=true}, B=AppendVector(A=@n12, B=@n13)))
  n15 = ComponentMask(None=@n14) {r=true}
  n16 = Multiply(A=ComponentMask(None=@n3) {b=true, g=true}, B=@n15)
  n17 = ComponentMask(None=@n14) {g=true}
  n18 = Multiply(A=ComponentMask(None=@n3) {b=true, r=true}, B=@n17)
  n19 = ComponentMask(None=@n3) {g=true, r=true}
  n20 = OneMinus(None=Max(A=@n15, B=@n17))
  FunctionOutput(None=Add(A=Add(A=@n16, B=@n18), B=Multiply(A=@n19, B=@n20))) {outputName=UVs}
  n21 = FunctionInput() {bUsePreviewValueAsDefault=true, inputName=Derivatives Scale, inputType=FunctionInput_Scalar, sortPriority=2, previewValue=<1 1 1 1>}
  n22 = Multiply(A=TextureCoordinate(), B=@n21)
  FunctionOutput(None=DDX(Value=@n22)) {outputName=DDX, sortPriority=1}
  FunctionOutput(None=DDY(Value=@n22)) {outputName=DDY, sortPriority=2}
  FunctionOutput(None=AppendVector(A=AppendVector(A=@n15, B=@n17), B=@n20)) {outputName=Mask, sortPriority=3}
 ]}
