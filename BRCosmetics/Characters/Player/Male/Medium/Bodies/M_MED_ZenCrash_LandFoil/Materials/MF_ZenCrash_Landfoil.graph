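# Material function for the ZenCrash LandFoil character body that layers a timed, reactive fire effect over input material attributes, altering emissive, base color and opacity.
# Timestamp parameters drive fade-in, hold and fade-out, alongside a UV1 wipe and SDF or texture masks for tattoo and belt.
# Panned, noise-warped fire on dithered pre-skinned triplanar or mirrored arm UVs, using explicit derivatives, feeds Fresnel color gradients, rims and outlines.
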MaterialFunction /BRCosmetics/Characters/Player/Male/Medium/Bodies/M_MED_ZenCrash_LandFoil/Materials/MF_ZenCrash_Landfoil
graph {nodes [
  Constant() {r=1}
  n1 = GetMaterialAttributes(None=NamedRerouteUsage()) {attributeGetTypes [
    #1
    #2
    #3
    #4
   ]}
  n2 = NamedRerouteUsage()
  n3 = Multiply(A=@n2, B=ScalarParameter() {defaultValue=2, parameterName=B_3}) {constB=202}
  n4 = NamedRerouteUsage()
  n5 = SmoothStep(Value=@n2) {constMax=0.1}
  FunctionOutput(None=SetMaterialAttributes!(
    MaterialAttributes=@n1.MaterialAttributes
    EmissiveColor=NamedRerouteUsage()
    BaseColor=LinearInterpolate(A=@n1.BaseColor, Alpha=@n3) {constB=0}
    Opacity=LinearInterpolate(A=@n4, Alpha=@n5) {constB=0}
   ) {attributeSetTypes [#1, #2, #3]})
  n6 = NamedRerouteUsage()
  n7 = ComponentMask(None=Add(A=TextureCoordinate() {coordinateIndex=1}) {constB=-0.9342}) {g=true}
  n8 = Multiply(A=@n7) {constB=-1}
  n9 = Ceil(None=OneMinus(None=NamedRerouteUsage()))
  n10 = ComponentMask(None=NamedRerouteUsage()) {r=true}
  n11 = Multiply(A=@n10) {constB=-1}
  n12 = LinearInterpolate(
   A=LinearInterpolate(A=@n7, B=Subtract(A=@n8), Alpha=@n9)
   B=LinearInterpolate(A=Subtract(A=@n11), B=@n10, Alpha=@n9)
   Alpha=NamedRerouteUsage()
  )
  n13 = Multiply(A=NamedRerouteUsage()) {constB=2}
  n14 = SmoothStep(Value=Saturate(None=Add(A=@n12, B=@n13))) {constMax=0.8}
  LinearInterpolate(
   A=NamedRerouteUsage()
   B=SmoothStep(Value=SmoothStep(Min=@n6, Value=@n14)) {constMax=0.8}
   Alpha=ScalarParameter() {defaultValue=1, group=ZenCrash_LandFoil, parameterName=UseUVWipeForReactivity}
  )
  n15 = NamedRerouteUsage()
  n16 = NamedRerouteUsage()
  n17 = Max(A=Max(A=Max(A=@n15, B=@n16), B=NamedRerouteUsage()), B=NamedRerouteUsage())
  n18 = ScalarParameter() {defaultValue=1, group=ZenCrash_LandFoil, parameterName=ReactivityIntensity}
  NamedRerouteDeclaration(None=Multiply(A=Multiply(A=@n17, B=@n18), B=NamedRerouteUsage())) {name=EffectMask, nodeColor=<rgba 0.8984 1 0 1>}
  MaterialExpressionMaterialFunctionCall_4 = MaterialFunctionCall!(
   "Normal (WS)"
   Texture Scale=ScalarParameter() {defaultValue=120, group=ZenCrash_LandFoil, parameterName=TriplanarScale}
   Triplanar Fade=ScalarParameter() {defaultValue=20, group=ZenCrash_LandFoil, parameterName=TriplanarFade}
   Derivatives Scale
  ) {materialFunction=&/MF_TriplanarDithered_PreSkinned}
  NamedRerouteDeclaration(None=@MaterialExpressionMaterialFunctionCall_4.DDY) {name=DDY, nodeColor=<rgba 0.0781 1 0 1>}
  NamedRerouteDeclaration(None=MaterialFunctionCall!(Time Mult=Constant() {r=0.3}) {materialFunction=/Game/Effects/Fort_Effects/MaterialFunctions/MF_Flicker}.Raw Flicker) {name=Flicker, nodeColor=<rgba 0.9688 0 1 1>}
  n19 = FunctionInput() {inputName=FXMask, sortPriority=4}
  NamedRerouteDeclaration(None=ChannelMaskParameter(None=NamedRerouteDeclaration(None=@n19) {name=FXMask, nodeColor=<rgba 0.0781 1 0 1>}) {group=ZenCrash_LandFoil, maskChannel=Blue, parameterName=BeltFXMask, defaultValue=<rgba 0 0 1 0>}.R) {name=BeltMask, nodeColor=<rgba 1 0.5625 0 1>}
  n20 = Sine(None=Multiply(A=Time()) {constB=0.2})
  n21 = Time()
  n22 = Subtract(
   A=ScalarParameter() {group=ZenCrash_LandFoil, parameterName=Timestamp}
   B=ScalarParameter() {group=ZenCrash_LandFoil, parameterName=TimestampDelay}
  )
  n23 = ScalarParameter() {defaultValue=1, group=ZenCrash_LandFoil, parameterName=TimestampIn}
  n24 = Saturate(None=Divide(A=Subtract(A=@n21, B=@n22), B=@n23) {desc=InLength})
  n25 = Add(A=@n23, B=ScalarParameter() {defaultValue=0.75, group=ZenCrash_LandFoil, parameterName=TimestampStay}) {constB=3, desc=eXTRAdURATION}
  n26 = Divide(A=Subtract(A=@n21, B=Add(A=@n22, B=@n25)), B=ScalarParameter() {defaultValue=1, group=ZenCrash_LandFoil, parameterName=TimestampOut}) {constB=3, desc=OutLength}
  n27 = NamedRerouteDeclaration(None=OneMinus(None=Saturate(None=@n26))) {name=TimestampOut, nodeColor=<rgba 0 0.0859 1 1>}
  NamedRerouteDeclaration(None=StaticSwitchParameter(True=Saturate(None=@n20), False=Multiply(A=@n24, B=@n27)) {group=ZenCrash_LandFoil, parameterName=DebugTimestamp}) {name=Timestamp, nodeColor=<rgba 1 0 0.9141 1>}
  n28 = Panner(Coordinate=NamedRerouteUsage()) {speedX=-0.25, speedY=0.05}
  n29 = NamedRerouteUsage()
  n30 = NamedRerouteUsage()
  n31 = NamedRerouteUsage()
  n32 = Subtract(A=NamedRerouteDeclaration(None=TextureSample!(UVs=@n28, Tex=@n29, "DDX(UVs)"=@n30, "DDY(UVs)"=@n31, Apply View MipBias) {mipValueMode=TMVM_Derivative, texture=/UVEditor/Textures/UVEditorColorGrid_Color}.R) {name=CloudNoise, nodeColor=<rgba 0 0.4609 1 1>}) {constB=0.5}
  NamedRerouteDeclaration(None=Add(A=NamedRerouteUsage(), B=Multiply(A=@n32) {constB=0.2})) {name=WarpedUVs, nodeColor=<rgba 1 0 0.2578 1>}
  n33 = NamedRerouteUsage()
  n34 = ScalarParameter() {defaultValue=2, desc=UV must be integer to avoid seam, group=ZenCrash_LandFoil, parameterName=UV_Tiling01}
  n35 = ScalarParameter() {defaultValue=-0.3, group=ZenCrash_LandFoil, parameterName=FireSpeedX01}
  n36 = ScalarParameter() {defaultValue=0.1, group=ZenCrash_LandFoil, parameterName=FireSpeedY01}
  n37 = TextureSample(
   UVs=Panner(Coordinate=Multiply(A=@n33, B=@n34), Speed=AppendVector(A=@n35, B=@n36)) {speedX=0.2, speedY=-0.2}
   Tex=NamedRerouteUsage()
   "DDX(UVs)"=NamedRerouteUsage()
   "DDY(UVs)"=NamedRerouteUsage()
  ) {
   mipValueMode=TMVM_Derivative
   samplerSource=SSM_Wrap_WorldGroupSettings
   texture=/BRCosmetics/Characters/Player/Male/Medium/Bodies/M_MED_EssayViewPier/Textures/T_FireNoisesPacked
  }
  NamedRerouteDeclaration(None=Multiply(A=@n37.G, B=ScalarParameter() {defaultValue=1, group=ZenCrash_LandFoil, parameterName=FireEmissiveIntensity})) {name=FirstLayerFire, nodeColor=<rgba 0.8281 1 0 1>}
  n38 = TextureObjectParameter() {
   group=ZenCrash_LandFoil
   parameterName=FireTexture
   texture=&T_FireNoisesPacked
  }
  NamedRerouteDeclaration(None=@n38) {name=FireTexture, nodeColor=<rgba 1 0.2578 0 1>}
  n39 = Saturate(None=Subtract(A=NamedRerouteUsage(), B=NamedRerouteUsage()))
  NamedRerouteDeclaration(None=Multiply(A=OneMinus(None=@n39), B=NamedRerouteUsage())) {name=BeltTips, nodeColor=<rgba 0 1 0.9531 1>}
  NamedRerouteDeclaration(None=Multiply(A=NamedRerouteUsage(), B=NamedRerouteUsage())) {name=EmissiveOUT, nodeColor=<rgba 1 0 0 1>}
  NamedRerouteDeclaration(None=FunctionInput() {inputName=Material Input, inputType=FunctionInput_MaterialAttributes}) {name=Material_IN, nodeColor=<rgba 0 0.3437 1 1>}
  n40 = VectorParameter() {group=ZenCrash_LandFoil, parameterName=BeltColor1, defaultValue=<rgba 3.5 0.2256 0 1>}
  n41 = VectorParameter() {group=ZenCrash_LandFoil, parameterName=BeltColor2, defaultValue=<rgba 3 2.018 0 1>}
  n42 = Fresnel() {baseReflectFraction=0, exponent=2}
  n43 = NamedRerouteUsage()
  n44 = NamedRerouteUsage()
  n45 = Add(A=Multiply(A=@n42, B=Add(A=@n42, B=@n43)), B=Multiply(A=@n44, B=@n44))
  n46 = NamedRerouteUsage()
  NamedRerouteDeclaration(None=Multiply(A=LinearInterpolate(A=@n40.RGB, B=@n41.RGB, Alpha=@n45), B=ComponentMask(None=@n46) {g=true})) {name=BeltColor, nodeColor=<rgba 1 0 0.9844 1>}
  NamedRerouteDeclaration(None=@n37.B) {name=FirePattern, nodeColor=<rgba 0.1016 0 1 1>}
  NamedRerouteDeclaration(None=@n1.Opacity) {name=OpacityIn, nodeColor=<rgba 0.4766 1 0 1>}
  NamedRerouteDeclaration(None=Multiply(A=NamedRerouteUsage(), B=ScalarParameter() {defaultValue=1, group=ZenCrash_LandFoil, parameterName=ReactivityOn})) {name=ComplexReactivity, nodeColor=<rgba 0 1 0.5781 1>}
  n47 = Convert(None=@MaterialExpressionMaterialFunctionCall_4.UVs) {
   convertInputs [{
     type=Vector2
     expressionInput {
      expression=&self:MaterialExpressionMaterialFunctionCall_4
      inputName=None
      mask=0
      maskA=0
      maskB=0
      maskG=0
      maskR=0
      outputIndex=0
     }
    }]
   convertOutputs [{type=Vector2}]
  }
  n48 = Multiply(A=Constant2Vector() {g=1, r=-1}, B=@n47)
  n49 = TextureCoordinate() {coordinateIndex=1}
  n50 = ConstantBiasScale(None=@n49) {bias=-0.5, scale=2}
  n51 = ComponentMask(None=@n50) {r=true}
  n52 = Multiply(A=@n51) {constB=-1}
  n53 = Step(Y=ComponentMask(None=@n49) {r=true}) {constX=0.5}
  n54 = NamedRerouteDeclaration(None=AppendVector(A=LinearInterpolate(A=@n52, B=@n51, Alpha=@n53), B=ComponentMask(None=@n50) {g=true})) {name=ArmsUVs, nodeColor=<rgba 1 0.0938 0 1>}
  n55 = Multiply(A=ComponentMask(None=LocalPosition() {localOrigin=InstancePreSkinning}.XYZ) {b=true}) {constB=0.01}
  n56 = ComponentMask(None=NamedRerouteDeclaration(None=VertexInterpolator(VS=@n55).PS) {name=PreSkinnedPos, nodeColor=<rgba 0 0.2266 1 1>}) {g=true}
  n57 = NamedRerouteDeclaration(None=OneMinus(None=Step(Y=@n56) {constX=1.04})) {name=TorsoMask, nodeColor=<rgba 0 1 0.4141 1>}
  NamedRerouteDeclaration(None=LinearInterpolate(A=LinearInterpolate(A=@n48, B=@n54, Alpha=@n57), B=@n48, Alpha=ScalarParameter() {group=ZenCrash_LandFoil, parameterName=UsePreSkinnedUVs})) {name=UVs, nodeColor=<rgba 0.4297 0 1 1>}
  n58 = ScalarParameter() {defaultValue=0.28, group=ZenCrash_LandFoil, parameterName=OutlineMin}
  n59 = NamedRerouteUsage()
  n60 = SmoothStep(Value=NamedRerouteUsage()) {constMin=0.5}
  n61 = Multiply(
   A=Subtract(A=SmoothStep(Min=@n58, Value=@n59) {constMin=0.4}, B=@n60)
   B=ScalarParameter() {defaultValue=2, group=ZenCrash_LandFoil, parameterName=OutlineIntensity}
  )
  n62 = ScalarParameter() {defaultValue=2, group=ZenCrash_LandFoil, parameterName=FireBorderPow}
  n63 = Saturate(None=Subtract(A=Power(Base=@n59, Exp=@n62), B=@n60))
  n64 = NamedRerouteUsage()
  n65 = Multiply(A=Multiply(A=@n63, B=Add(A=@n63, B=@n64)), B=ScalarParameter() {defaultValue=1, group=ZenCrash_LandFoil, parameterName=FireBorderIntensity})
  NamedRerouteDeclaration(None=Max(A=Saturate(None=@n61), B=Saturate(None=@n65))) {name=FireRim, nodeColor=<rgba 0 1 0.9063 1>}
  n66 = ScalarParameter() {defaultValue=0.35, group=ZenCrash_LandFoil, parameterName=BaseColorFresnel}
  n67 = Multiply(A=@n66) {constB=0.1}
  n68 = Fresnel(ExponentIn=LinearInterpolate(A=Subtract(A=@n66, B=@n67) {constB=0.1}, B=@n66, Alpha=NamedRerouteUsage()) {constA=1}) {baseReflectFraction=0, exponent=0.75}
  n69 = Subtract(A=NamedRerouteUsage()) {constB=0.5}
  n70 = ScalarParameter() {defaultValue=-0.5, group=ZenCrash_LandFoil, parameterName=FirePatternColorIntensity}
  n71 = Multiply(A=@n68, B=Add(A=@n68, B=Multiply(A=@n69, B=@n70)))
  n72 = LinearInterpolate(Alpha=@n71) {constA=0.25}
  n73 = NamedRerouteUsage()
  n74 = Add(A=Max(A=@n71, B=Multiply(A=@n72, B=@n73)), B=NamedRerouteUsage())
  n75 = NamedRerouteDeclaration(None=LinearInterpolate(
    A=VectorParameter() {group=ZenCrash_LandFoil, parameterName=Color1, defaultValue=<rgba 1 0.9744 0.5208 1>}.RGB
    B=VectorParameter() {group=ZenCrash_LandFoil, parameterName=Color2, defaultValue=<rgba 1 0.2839 0 1>}.RGB
    Alpha=Saturate(None=@n74)
   )) {name=FresnelColor, nodeColor=<rgba 0.125 0 1 1>}
  n76 = Multiply(A=VectorParameter() {group=ZenCrash_LandFoil, parameterName=Outlinecolor, defaultValue=<rgba 2 0.0053 0 1>}.RGB, B=NamedRerouteUsage())
  NamedRerouteDeclaration(None=LinearInterpolate(A=Add(A=@n75, B=@n76), B=NamedRerouteUsage(), Alpha=NamedRerouteUsage())) {name=Outcolor, nodeColor=<rgba 1 0.5859 0 1>}
  n77 = TextureSample!(UVs, Tex, Apply View MipBias) {
   constCoordinate=1
   samplerType=SAMPLERTYPE_LinearGrayscale
   texture=/BRCosmetics/Characters/Player/Male/Medium/Bodies/M_MED_ZenCrash_LandFoil/Textures/T_ZenCrash_SDF
  }
  n78 = TextureSampleParameter2D() {
   constCoordinate=1
   group=ZenCrash_LandFoil
   parameterName=EffectMaskTexture
   samplerType=SAMPLERTYPE_Masks
   texture=/BRCosmetics/Characters/Player/Male/Medium/Bodies/M_MED_ZenCrash_LandFoil/Textures/T_ZenCrash_LandFoil_Body_UV2_FX
  }
  n79 = Power(Base=@n78.B, Exp=ScalarParameter() {defaultValue=1, group=ZenCrash_LandFoil, parameterName=Exp})
  n80 = ScalarParameter() {defaultValue=1.05, group=ZenCrash_LandFoil, parameterName=B_4}
  n81 = NamedRerouteUsage()
  NamedRerouteDeclaration(None=Power(
    Base=StaticSwitchParameter(True=@n77.R, False=@n79) {defaultValue=true, group=ZenCrash_LandFoil, parameterName=UseSDF}
    Exp=LinearInterpolate(B=@n80, Alpha=@n81) {constA=1}
   )) {name=TattooSDF, nodeColor=<rgba 0 0.0391 1 1>}
  n82 = NamedRerouteUsage()
  n83 = Multiply(A=Ceil(None=@n82), B=SphereMask(A=OneMinus(None=@n82), B=NamedRerouteUsage()) {attenuationRadius=0.5, hardnessPercent=0})
  n84 = NamedRerouteUsage()
  NamedRerouteDeclaration(None=Saturate(None=Subtract(A=@n83, B=@n84))) {name=OutlineFire, nodeColor=<rgba 0.2422 1 0 1>}
  NamedRerouteDeclaration(None=FunctionInput() {inputName=M, sortPriority=1}) {name=M_Mask, nodeColor=<rgba 0.0781 1 0 1>}
  NamedRerouteDeclaration(None=@MaterialExpressionMaterialFunctionCall_4.DDX) {name=DDX, nodeColor=<rgba 0.2422 0 1 1>}
  NamedRerouteDeclaration(None=@n79) {name=EmissiveMask, nodeColor=<rgba 0 1 0.8594 1>}
 ]}
