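# Parent eye material for the ZenCrash LandFoil skin, adding a reactive, timed emissive glow to the shared eye shading.
# MF_eyes_2023 output gains the head emissive texture scaled by ReactivityOn and a time envelope, then passes through MF_Eyelash and MF_TechArtOverrides.
# Timestamp parameters drive the envelope's delayed fade-in, hold and fade-out against Time.
# The DebugTimestamp static switch swaps that envelope for a looping clamped sine.
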Material /BRCosmetics/Characters/Player/Male/Medium/Bodies/M_MED_ZenCrash_LandFoil/Materials/M_ZenCrash_Landfoil_Eyes_Parent
graph {
 properties {
  bUseMaterialAttributes=true
  bUsedWithMeshDeformer=true
  bUsedWithSkeletalMesh=true
  pixelDepthOffsetMode=PDOM_Legacy
  shadingModel=MSM_FromMaterialExpression
  usedShadingModels=Default Lit | Subsurface
 }
 nodes [
  Constant() {r=1}
  n1 = GetMaterialAttributes(None=MaterialFunctionCall() {materialFunction=/Game/Global/Materials/Character_Parent_2023/Eye_Parent_2023/MF_eyes_2023}.Result) {attributeGetTypes [#1]}
  n2 = NamedRerouteUsage()
  n3 = ScalarParameter() {defaultValue=1, group=ZenCrash_LandFoil, parameterName=ReactivityOn}
  n4 = Add(
   A=@n1.Emissive Color
   B=Multiply(
    A=TextureSample() {texture=/BRCosmetics/Characters/Player/Male/Medium/Bodies/M_MED_ZenCrash_LandFoil/Textures/T_ZenCrash_Landfoil_Head_E}.RGB
    B=Multiply(A=@n2, B=@n3)
   )
  )
  MaterialFunctionCall(In=MaterialFunctionCall!(Material=SetMaterialAttributes(MaterialAttributes=@n1.MaterialAttributes, Emissive Color=@n4) {attributeSetTypes [#1]}) {materialFunction=/Game/Global/Materials/Character_Parent_2023/Eye_Parent_2023/MF_Eyelash}.Result) {materialFunction=&MF_TechArtOverrides}
  n5 = Sine(None=Multiply(A=Time()) {constB=0.2})
  n6 = Time()
  n7 = Subtract(
   A=ScalarParameter() {group=ZenCrash_LandFoil, parameterName=Timestamp}
   B=ScalarParameter() {group=ZenCrash_LandFoil, parameterName=TimestampDelay}
  )
  n8 = ScalarParameter() {defaultValue=1, group=ZenCrash_LandFoil, parameterName=TimestampIn}
  n9 = Saturate(None=Divide(A=Subtract(A=@n6, B=@n7), B=@n8) {desc=InLength})
  n10 = Add(A=@n8, B=ScalarParameter() {defaultValue=0.75, group=ZenCrash_LandFoil, parameterName=TimestampStay}) {constB=3, desc=eXTRAdURATION}
  n11 = Divide(A=Subtract(A=@n6, B=Add(A=@n7, B=@n10)), B=ScalarParameter() {defaultValue=1, group=ZenCrash_LandFoil, parameterName=TimestampOut}) {constB=3, desc=OutLength}
  n12 = NamedRerouteDeclaration(None=OneMinus(None=Saturate(None=@n11))) {name=TimestampOut, nodeColor=<rgba 0 0.0859 1 1>}
  NamedRerouteDeclaration(None=StaticSwitchParameter(True=Saturate(None=@n5), False=Multiply(A=@n9, B=@n12)) {group=ZenCrash_LandFoil, parameterName=DebugTimestamp}) {name=Timestamp, nodeColor=<rgba 1 0 0.9141 1>}
 ]
}
