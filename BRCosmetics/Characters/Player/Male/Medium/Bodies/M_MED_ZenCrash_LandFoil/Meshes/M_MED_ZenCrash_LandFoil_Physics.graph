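# Physics asset for the ZenCrash LandFoil medium male player body mesh, simulating its dangling necklace.
# Capsule bodies cover the upper spine bones, a necklace base and three dynamic necklace bones.
# Angular-limited constraints chain the necklace from its base through the main link to two further links, with swing loosening toward the tip.
# The spine capsules have no constraints, apparently serving as torso collision for the necklace.
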PhysicsAsset /BRCosmetics/Characters/Player/Male/Medium/Bodies/M_MED_ZenCrash_LandFoil/Meshes/M_MED_ZenCrash_LandFoil_Physics
bodies {
 base_necklace_main_02 [{shapeType=Capsule, length=0.505, radius=0.505, shapeName=base_necklace_main_02_capsule, rotation=<rot 0 -90 0>}]
 dyn_necklace_01 [{shapeType=Capsule, length=11.6635, radius=2.1728, shapeName=dyn_necklace_01_capsule, center=<4.5 0 2.1>, rotation=<rot 90 0 90>}]
 dyn_necklace_02 [{shapeType=Capsule, length=8.7159, radius=3.1029, shapeName=dyn_necklace_02_capsule, center=<6.2495 0 1.3157>, rotation=<rot 72.9909 0 90>}]
 dyn_necklace_main_01 [{
   shapeType=Capsule
   length=11.0033
   radius=2.9967
   shapeName=dyn_necklace_main_01_capsule
   center=<11.583 0 5.4424>
   rotation=<rot 82.189 0 90>
  }]
 spine_03 [{shapeType=Capsule, length=17.8027, radius=15.2406, center=<11.5387 -2.0859 0.3254>}]
 spine_04 [{shapeType=Capsule, length=17.8904, radius=14.4089, center=<4.8656 -1.2747 0.3443>, rotation=<rot 0 -90 0>}]
 spine_05 [{shapeType=Capsule, length=20.6974, radius=14.4504, center=<0.7939 -4.004 -0.2733>, rotation=<rot 0 -90 0>}]
}
constraints [
 {bone1Name=dyn_necklace_main_01, bone2Name=base_necklace_main_02, swing1LimitDegrees=10, swing2LimitDegrees=10, twistLimitDegrees=5}
 {bone1Name=dyn_necklace_01, bone2Name=dyn_necklace_main_01, swing1LimitDegrees=10, swing2LimitDegrees=30, twistLimitDegrees=5}
 {bone1Name=dyn_necklace_02, bone2Name=dyn_necklace_01, swing1LimitDegrees=15, swing2LimitDegrees=40, twistLimitDegrees=5}
]
