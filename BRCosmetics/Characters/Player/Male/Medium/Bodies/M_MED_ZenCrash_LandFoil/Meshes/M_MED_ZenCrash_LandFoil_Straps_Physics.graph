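# Physics asset for the straps on the M_MED_ZenCrash_LandFoil male medium player cosmetic body.
# Left and right strap chains each link five capsule bodies from a small base capsule through limited-swing constraints with zero-degree twist limits.
# The lower right-side joints allow slightly wider swing on one axis than the rest.
# Unconstrained pelvis and thigh capsules sit alongside the chains, apparently as collision volumes for the straps.
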PhysicsAsset /BRCosmetics/Characters/Player/Male/Medium/Bodies/M_MED_ZenCrash_LandFoil/Meshes/M_MED_ZenCrash_LandFoil_Straps_Physics
bodies {
 base_strap_01_lf [{shapeType=Capsule, length=0.505, radius=0.505, shapeName=base_strap_01_lf_capsule, rotation=<rot 0 -90 0>}]
 base_strap_01_rt [{shapeType=Capsule, length=0.505, radius=0.505, shapeName=base_strap_01_rt_capsule, rotation=<rot 0 -90 0>}]
 dyn_strap_01_lf [{shapeType=Capsule, length=1.6799, radius=0.5, shapeName=dyn_strap_01_lf_capsule, center=<2 0 0>, rotation=<rot 90 0 0>}]
 dyn_strap_01_rt [{shapeType=Capsule, length=3.3199, radius=1.5, shapeName=dyn_strap_01_rt_capsule, center=<2 -2 0>}]
 dyn_strap_02_lf [{shapeType=Capsule, length=4.7098, radius=1.4, shapeName=dyn_strap_02_lf_capsule, center=<3.3597 -1 0>, rotation=<rot 9.2871 0 0>}]
 dyn_strap_02_rt [{shapeType=Capsule, length=3.3199, radius=1.5, shapeName=dyn_strap_02_rt_capsule, center=<3.5 -1.8 0>}]
 dyn_strap_03_lf [{shapeType=Capsule, length=4.7098, radius=1.4, shapeName=dyn_strap_03_lf_capsule, center=<3 -1 0>}]
 dyn_strap_03_rt [{shapeType=Capsule, length=3.3971, radius=1.4466, shapeName=dyn_strap_03_rt_capsule, center=<2.5 -1.2 0>}]
 dyn_strap_04_lf [{shapeType=Capsule, length=4.7098, radius=1.4, shapeName=dyn_strap_04_lf_capsule, center=<2 -1 0>}]
 dyn_strap_04_rt [{shapeType=Capsule, length=3.3971, radius=1.4466, shapeName=dyn_strap_04_rt_capsule, center=<2 -1.2 0>}]
 dyn_strap_05_lf [{shapeType=Capsule, length=4.7098, radius=1.4, shapeName=dyn_strap_05_lf_capsule, center=<2 -1 0>}]
 dyn_strap_05_rt [{shapeType=Capsule, length=3.3971, radius=1.4466, shapeName=dyn_strap_05_rt_capsule, center=<2 -1.2 0>}]
 pelvis [
  {shapeType=Capsule, length=16.3109, radius=10.2093, shapeName=pelvis_capsule}
  {shapeType=Capsule, length=19.8048, radius=10.1952, shapeName=pelvis_capsule, center=<-6.4179 1.7646 0.0048>, rotation=<rot 90 -6.7907 0>}
 ]
 thigh_l [{shapeType=Capsule, length=28.2476, radius=9.064, shapeName=thigh_l_capsule, center=<-15.9122 0.7104 0>, rotation=<rot 90 4.5471 0>}]
 thigh_r [{shapeType=Capsule, length=28.2476, radius=9.064, shapeName=thigh_l_capsule, center=<15.9122 -0.7104 0>, rotation=<rot 90 4.5471 0>}]
}
constraints [
 {bone1Name=dyn_strap_01_lf, bone2Name=base_strap_01_lf, swing1LimitDegrees=20, swing2LimitDegrees=20, twistLimitDegrees=0}
 {bone1Name=dyn_strap_02_lf, bone2Name=dyn_strap_01_lf, swing1LimitDegrees=20, swing2LimitDegrees=20, twistLimitDegrees=0}
 {bone1Name=dyn_strap_03_lf, bone2Name=dyn_strap_02_lf, swing1LimitDegrees=20, swing2LimitDegrees=20, twistLimitDegrees=0}
 {bone1Name=dyn_strap_04_lf, bone2Name=dyn_strap_03_lf, swing1LimitDegrees=20, swing2LimitDegrees=20, twistLimitDegrees=0}
 {bone1Name=dyn_strap_05_lf, bone2Name=dyn_strap_04_lf, swing1LimitDegrees=20, swing2LimitDegrees=20, twistLimitDegrees=0}
 {bone1Name=dyn_strap_01_rt, bone2Name=base_strap_01_rt, swing1LimitDegrees=20, swing2LimitDegrees=20, twistLimitDegrees=0}
 {bone1Name=dyn_strap_02_rt, bone2Name=dyn_strap_01_rt, swing1LimitDegrees=20, swing2LimitDegrees=20, twistLimitDegrees=0}
 {bone1Name=dyn_strap_03_rt, bone2Name=dyn_strap_02_rt, swing1LimitDegrees=30, swing2LimitDegrees=20, twistLimitDegrees=0}
 {bone1Name=dyn_strap_04_rt, bone2Name=dyn_strap_03_rt, swing1LimitDegrees=30, swing2LimitDegrees=20, twistLimitDegrees=0}
 {bone1Name=dyn_strap_05_rt, bone2Name=dyn_strap_04_rt, swing1LimitDegrees=30, swing2LimitDegrees=20, twistLimitDegrees=0}
]
