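# Control Rig for the ZenCrash LandFoil male character's face accessory part that adjusts the hair ponytail bones according to head orientation.
# Three spherical pose readers on the head bone, run through a sequence, weight Modify Transforms nodes.
# The first, parented to neck_02, passes through a curve remap to rotate and Z-squash the ponytail.
# The other two rotate the ponytail base, and all three readers have debug drawing enabled.
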ControlRigBlueprint /BRCosmetics/Characters/Player/Male/Medium/Bodies/M_MED_ZenCrash_LandFoil/Meshes/Parts/M_MED_ZenCrash_LandFoil_FaceAcc_CtrlRig
graphs [
 RigVMFunctionLibrary {links [], nodes []}
 RigVMFunction_Sequence_1_ContainedGraph {nodes [
   n1 = RigVMFunctionEntryNode()
   RigVMFunction_Sequence = &RigVMFunction_Sequence(ExecuteContext=@n1.ExecuteContext)
   n2 = &RigVMFunction_Sequence(ExecuteContext=@RigVMFunction_Sequence.B)
   n3 = &RigVMFunction_Sequence(ExecuteContext=@n2.B)
   RigVMFunctionReturnNode(A=@RigVMFunction_Sequence.A, B=@n2.A, C=@n3.A, D=@n3.B)
  ]}
 RigVMModel {nodes [
   RigUnit_BeginExecution = &RigUnit_BeginExecution()
   RigVMFunction_Sequence_1 = RigVMAggregateNode(ExecuteContext=@RigUnit_BeginExecution.ExecutePin)
   n1 = &RigUnit_SphericalPoseReader(
    ActiveRegionScaleFactors=(PositiveWidth=3.25,NegativeWidth=3.25,PositiveHeight=14,NegativeHeight=7)
    ActiveRegionSize=0.02
    Debug=(bDrawDebug=true,bDraw2D=false,bDrawLocalAxes=true,DebugScale=50,DebugSegments=13,DebugThickness=0.25)
    DriverAxis=(X=0,Y=-1,Z=0)
    DriverItem=(Type=Bone,Name="head")
    FalloffRegionScaleFactors=(PositiveWidth=1,NegativeWidth=0.45,PositiveHeight=1,NegativeHeight=1.5)
    FalloffSize=0.3
    FlipHeightScaling=False
    FlipWidthScaling=False
    OptionalParentItem=(Type=Bone,Name="neck_02")
    RotationOffset=(X=-90,Y=0,Z=-54)
    ExecutePin=@RigVMFunction_Sequence_1.A
   )
   n2 = &RigVMFunction_AnimEvalRichCurve(
    Curve=(EditorCurveData=(Keys=((InterpMode=RCIM_Cubic,TangentMode=RCTM_User,ArriveTangent=0.2335,LeaveTangent=0.2335),(InterpMode=RCIM_Cubic,TangentMode=RCTM_User,Time=1,Value=1,ArriveTangent=0.7873,LeaveTangent=0.7873))))
    SourceMinimum=0
    TargetMinimum=0
    Value=@n1.OutputParam
   )
   > &RigUnit_ModifyTransforms(
    ItemToModify="((Item=(Type=Bone,Name=\"base_ponytail01_cr\"),Transform=(Rotation=(X=0,Y=0.5373,Z=0,W=0.8434),Translation=(X=0,Y=0,Z=0),Scale3D=(X=1,Y=1,Z=0.1))),(Item=(Type=Bone,Name=\"dyn_ponytailLower_cr\"),Transform=(Rotation=(X=0,Y=0.342,Z=0,W=0.9397),Translation=(X=0,Y ...(truncated)"
    WeightMinimum=0
    Weight=@n2.Result
   )
   n3 = &RigUnit_SphericalPoseReader(
    ActiveRegionScaleFactors=(PositiveWidth=1,NegativeWidth=1,PositiveHeight=1,NegativeHeight=1)
    ActiveRegionSize=0.025
    Debug=(bDrawDebug=true,bDraw2D=false,bDrawLocalAxes=true,DebugScale=50,DebugSegments=13,DebugThickness=0.25)
    DriverAxis=(X=0,Y=-1,Z=0)
    DriverItem=(Type=Bone,Name="head")
    FalloffRegionScaleFactors=(PositiveWidth=1,NegativeWidth=1,PositiveHeight=1,NegativeHeight=1)
    FalloffSize=0.11
    FlipHeightScaling=False
    FlipWidthScaling=False
    OptionalParentItem=(Type=Bone,Name="None")
    RotationOffset=(X=-80,Y=0,Z=-34)
    ExecutePin=@RigVMFunction_Sequence_1.B
   )
   > &RigUnit_ModifyTransforms(
    ItemToModify=((Item=(Type=Bone,Name="base_ponytail01_cr"),Transform=(Rotation=(X=0.0449,Y=0.1677,Z=-0.2549,W=0.9513),Translation=(X=0,Y=0,Z=0),Scale3D=(X=1,Y=1,Z=1))))
    WeightMinimum=0
    Weight=@n3.OutputParam
   )
   > n4 = &RigUnit_SphericalPoseReader(
    ActiveRegionScaleFactors=(PositiveWidth=1,NegativeWidth=1,PositiveHeight=1,NegativeHeight=1)
    ActiveRegionSize=0.02
    Debug=(bDrawDebug=true,bDraw2D=false,bDrawLocalAxes=true,DebugScale=50,DebugSegments=13,DebugThickness=0.25)
    DriverAxis=(X=0,Y=-1,Z=0)
    DriverItem=(Type=Bone,Name="head")
    FalloffRegionScaleFactors=(PositiveWidth=1,NegativeWidth=1,PositiveHeight=1,NegativeHeight=1)
    FalloffSize=0.1
    FlipHeightScaling=False
    FlipWidthScaling=False
    OptionalParentItem=(Type=Bone,Name="None")
    RotationOffset=(X=-65,Y=0,Z=-15)
   )
   > &RigUnit_ModifyTransforms(
    ItemToModify=((Item=(Type=Bone,Name="base_ponytail01_cr"),Transform=(Rotation=(X=0,Y=0.1736,Z=0,W=0.9848),Translation=(X=0,Y=0,Z=0),Scale3D=(X=1,Y=1,Z=1))))
    WeightMinimum=0
    Weight=@n4.OutputParam
   )
  ]}
]
hierarchy [
 root {loc=<0 0 0>, rot=<0 0 0 1>}
 pelvis {parent=root, loc=<0 -0.7071 93.1191>, rot=<0.0127 0.707 -0.0127 -0.707>}
 spine_01 {parent=pelvis, loc=<6.7494 -0.6388 0>, rot=<0 0 0.034 -0.9994>}
 spine_02 {parent=spine_01, loc=<7.2518 -0.035 0>, rot=<0 0 0.0238 -0.9997>}
 spine_03 {parent=spine_02, loc=<6.0104 0 0>, rot=<0 0 0.0149 -0.9999>}
 spine_04 {parent=spine_03, loc=<8.1024 0 0>, rot=<0 0 -0.0132 -0.9999>}
 spine_05 {parent=spine_04, loc=<10.0146 0 0>, rot=<0 0 -0.0015 -1>}
 neck_01 {parent=spine_05, loc=<11.3138 0.1529 0>, rot=<0 0 -0.2003 -0.9797>}
 neck_02 {parent=neck_01, loc=<4.5627 0 0>, rot=<0 0 0.0504 -0.9987>}
 head {parent=neck_02, loc=<5.3914 0 0>, rot=<0 0 0.115 -0.9934>}
 base_faceframe02_rt {parent=head, loc=<5.9041 3.2923 7.6387>, rot=<-0.0432 -0.9971 0.0544 -0.0312>}
 dyn_faceframe02_rt {parent=base_faceframe02_rt, loc=<0 0 0>, rot=<0 0 0 1>}
 base_faceframe01_rt {parent=head, loc=<4.7737 5.6768 7.2704>, rot=<-0.0491 -0.9896 0.1209 -0.0603>}
 dyn_faceframe01_rt {parent=base_faceframe01_rt, loc=<0 0 0>, rot=<0 0 0 1>}
 base_hairMain_lf {parent=head, loc=<6.8417 -4.0978 -6.7508>, rot=<0.1277 0.2444 0.8177 -0.5052>}
 dyn_hairMain_lf {parent=base_hairMain_lf, loc=<0 0 0>, rot=<0 0 0 1>}
 base_hairMain_rt {parent=head, loc=<6.7415 -4.0643 6.5694>, rot=<0.4778 -0.8107 -0.2892 -0.1754>}
 dyn_hairMain_rt {parent=base_hairMain_rt, loc=<0 0 0>, rot=<0 0 0 1>}
 base_hairMain_cr {parent=head, loc=<9.2259 -7.7957 0.3342>, rot=<-0.084 0.7015 0.7027 -0.0843>}
 dyn_hairMain_cr {parent=base_hairMain_cr, loc=<0 0 0>, rot=<0 0 0 1>}
 base_ponytail01_cr {parent=head, loc=<2.0589 -8.0937 0>, rot=<-0.3326 0.624 0.624 -0.3326>}
 dyn_ponytail01_cr {parent=base_ponytail01_cr, loc=<0 0 0>, rot=<0 0 0 1>}
 dyn_ponytail02_cr {parent=dyn_ponytail01_cr, loc=<5.3065 0 0.0018>, rot=<0 0 0 1>}
 dyn_ponytailLower_cr {parent=dyn_ponytail02_cr, loc=<-0.8429 0 2.1322>, rot=<0 0.0443 0 -0.999>}
 dyn_ponytailUpper_cr {parent=dyn_ponytail02_cr, loc=<-1.3512 0 -2.0205>, rot=<0 -0.326 0 -0.9454>}
 base_faceframe01_lf {parent=head, loc=<7.1216 5.7858 -6.3306>, rot=<-0.0041 -0.2391 0.9644 0.1128>}
 dyn_faceframe01_lf {parent=base_faceframe01_lf, loc=<0 0 0>, rot=<0 0 0 1>}
 base_faceframe02_lf {parent=head, loc=<4.86 2.614 -7.3272>, rot=<0.0616 -0.0923 0.9913 0.0712>}
 dyn_faceframe02_lf {parent=base_faceframe02_lf, loc=<0 0 0>, rot=<0 0 0 1>}
 base_bangs_cr {parent=head, loc=<12.2367 8.8436 0.8887>, rot=<0.0446 -0.8067 0.5132 0.2897>}
 dyn_bangs_cr {parent=base_bangs_cr, loc=<0 0 0>, rot=<0 0 0 1>}
 base_bangs01_lf {parent=head, loc=<12.149 6.2632 -4.404>, rot=<-0.0588 -0.444 0.8042 0.3907>}
 dyn_bangs01_lf {parent=base_bangs01_lf, loc=<0 0 0>, rot=<0 0 0 1>}
 base_bangs02_lf {parent=head, loc=<9.8121 5.7764 -5.7156>, rot=<-0.0694 -0.3165 0.9089 0.2625>}
 dyn_bangs02_lf {parent=base_bangs02_lf, loc=<0 0 0>, rot=<0 0 0 1>}
 base_hairTop_lf {parent=head, loc=<13.3358 1.7048 -5.5299>, rot=<-0.2203 -0.1053 0.9687 0.0447>}
 dyn_hairTop_lf {parent=base_hairTop_lf, loc=<0 0 0>, rot=<0 0 0 1>}
 base_hairTop01_rt {parent=head, loc=<12.0428 5.119 5.8413>, rot=<-0.0609 -0.9464 0.2605 0.1807>}
 dyn_hairTop01_rt {parent=base_hairTop01_rt, loc=<0 0 0>, rot=<0 0 0 1>}
 base_hairTop02_rt {parent=head, loc=<15.01 -0.3187 1.7822>, rot=<-0.0567 0.8061 0.1252 -0.5756>}
 dyn_hairTop02_rt {parent=base_hairTop02_rt, loc=<0 0 0>, rot=<0 0 0 1>}
 base_hairTop_cr {parent=head, loc=<14.5582 1.8577 -2.7284>, rot=<0.115 0.6039 -0.278 -0.7381>}
 dyn_hairTop_cr {parent=base_hairTop_cr, loc=<0 0 0>, rot=<0 0 0 1>}
 base_hairCowlick02_cr {parent=head, loc=<14.0409 -4.6713 -1.5997>, rot=<-0.1323 0.649 0.5524 -0.5061>}
 dyn_hairCowlick02_cr {parent=base_hairCowlick02_cr, loc=<0 0 0>, rot=<0 0 0 1>}
 base_hairCowlick01_cr {parent=head, loc=<14.4847 -2.5533 -2.7532>, rot=<-0.4778 0.2851 0.8219 -0.1224>}
 dyn_hairCowlick01_cr {parent=base_hairCowlick01_cr, loc=<0 0 0>, rot=<0 0 0 1>}
 clavicle_r {parent=spine_05, loc=<4.3575 -0.1466 2.3223>, rot=<0.7289 0.0465 -0.6829 0.009>}
]
