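# Physics asset for the face accessory part of the male medium ZenCrash LandFoil character body, covering its ponytail.
# It has capsule bodies on a ponytail base bone, a dynamic ponytail bone and spine_05, whose long capsule likely acts as an upper-body collider for the hair.
# A single constraint joins the dynamic bone to the base with twist locked and small swing limit angles set; the spine body is unconstrained.
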PhysicsAsset /BRCosmetics/Characters/Player/Male/Medium/Bodies/M_MED_ZenCrash_LandFoil/Meshes/Parts/M_MED_ZenCrash_LandFoil_FaceAcc_Physics
bodies {
 base_ponytail01_cr [{shapeType=Capsule, length=0.5, radius=0.5, shapeName=base_ponytail01_cr_capsule}]
 dyn_ponytail01_cr [{shapeType=Capsule, length=2.5, radius=4.3, shapeName=dyn_ponytail01_cr_capsule, center=<6.2196 0 0.4865>, rotation=<rot 90 0 0>}]
 spine_05 [{shapeType=Capsule, length=39.9032, radius=7.8, shapeName=spine_05_capsule, center=<4.7917 0 0>, rotation=<rot 0 -80.0005 -180>}]
}
constraints [{
  bone1Name=dyn_ponytail01_cr
  bone2Name=base_ponytail01_cr
  swing1LimitDegrees=10
  swing2LimitDegrees=15
  twistLimitDegrees=45
  twistMotion=Locked
 }]
