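# Post-process-style Anim Blueprint for the ZenCrash LandFoil face accessory, adding secondary motion to ponytail, bangs and other hair bones.
# It copies the attached parent's pose through a Control Rig, a head-space RigidBody node, then chained spring-based AnimDynamics nodes, one per bone.
# Each update, a nativized function sets gravity and rigid-body setting variables from a parameter asset.
# Zipline enter/exit events zero and restore two AnimDynamics scale vectors.
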AnimBlueprint /BRCosmetics/Characters/Player/Male/Medium/Bodies/M_MED_ZenCrash_LandFoil/Meshes/Parts/M_MED_ZenCrash_LandFoil_FaceAcc_AnimBP
target_skeleton=&/M_MED_ZenCrash_LandFoil_FaceAcc_Skeleton
graphs [
 AnimGraph {nodes [
   n1 = Control Rig(Source=Copy Pose From Mesh() {bUseAttachedParent=true}) {controlRigAssetReference {blueprintRigClass=&/M_MED_ZenCrash_LandFoil_FaceAcc_CtrlRig.M_MED_ZenCrash_LandFoil_FaceAcc_CtrlRig_C}}
   n2 = RigidBody!(
    ComponentPose=Local To Component(LocalPose=@n1)
    Alpha
    OverrideWorldGravity:Vector="0,0,0"
    ExternalForce
    SimSpaceSettings:Sim Space Settings Structure=(WorldAlpha=1,VelocityScaleZ=1,DampingAlpha=1,MaxLinearVelocity=500,MaxAngularVelocity=40,MaxLinearAcceleration=500,MaxAngularAcceleration=93,ExternalLinearDragV=(X=0,Y=1,Z=0),ExternalLinearVelocity=(X=0,Y=0,Z=0),ExternalAngularVelocity=(X=0,Y=0,Z=0))
   ) -> (Pose) {
    bOverrideWorldGravity=true
    overridePhysicsAsset=&/M_MED_ZenCrash_LandFoil_FaceAcc_Physics
    simulationSpace=BaseBoneSpace
    baseBoneRef {boneName=head}
    simSpaceSettings {maxAngularAcceleration=93, maxAngularVelocity=40, maxLinearAcceleration=500, maxLinearVelocity=500, worldAlpha=1, externalLinearDragV {y=1}}
   }
   n3 = Anim Dynamics - Bone: dyn_ponytail02_cr!(
    ComponentPose=@n2
    Alpha=1
    GravityOverride:Vector="0,0,0"
    ComponentLinearAccScale:Vector="1,1,1"
    ComponentLinearVelScale:Vector="1,1,1"
    ExternalForce="0,0,0"
   ) -> (Pose) {
    angularDampingOverride=0.9999
    angularSpringConstant=2500
    bAngularSpring=true
    bOverrideAngularDamping=true
    bUseGravityOverride=true
    gravityScale=0.35
    simulationSpace=RootRelative
    boundBone {boneName=dyn_ponytail02_cr}
    componentLinearAccScale=<1 1 1>
    componentLinearVelScale=<1 1 1>
    gravityOverride {z=-980}
    physicsBodyDefinitions [{
      boundBone {boneName=dyn_ponytail02_cr}
      boxExtents=<40 40 40>
      constraintSetup {angularLimitsMax=<0 8 5>, angularLimitsMin=<0 -8 -5>, angularTarget=<1 0 0>}
      localJointOffset=<3.5 0 0>
     }]
    simSpaceSettings {maxAngularAcceleration=93, maxAngularVelocity=40, simSpaceAngularAlpha=0.2}
   }
   n4 = Anim Dynamics - Bone: dyn_faceframe01_rt!(
    ComponentPose=@n3
    Alpha=1
    GravityOverride:Vector="0,0,0"
    ComponentLinearAccScale:Vector="0.3,0.3,0.3"
    ComponentLinearVelScale:Vector="0.3,0.3,0.3"
    ExternalForce="0,0,0"
   ) -> (Pose) {
    angularDampingOverride=0.9999
    angularSpringConstant=2500
    bAngularSpring=true
    bOverrideAngularDamping=true
    bUseGravityOverride=true
    gravityScale=0.35
    simulationSpace=RootRelative
    boundBone {boneName=dyn_faceframe01_rt}
    componentAppliedLinearAccClamp=<3000 3000 3000>
    componentLinearAccScale=<0.3 0.3 0.3>
    componentLinearVelScale=<0.3 0.3 0.3>
    gravityOverride {z=-980}
    physicsBodyDefinitions [{
      boundBone {boneName=dyn_faceframe01_rt}
      boxExtents=<27 27 27>
      constraintSetup {angularLimitsMax=<0 8 5>, angularLimitsMin=<0 -8 -5>, angularTarget=<1 0 0>}
      localJointOffset=<3.2 0 0>
     }]
    simSpaceSettings {maxAngularAcceleration=50, maxAngularVelocity=100, simSpaceAngularAlpha=0.2}
   }
   n5 = Anim Dynamics - Bone: dyn_faceframe02_rt!(
    ComponentPose=@n4
    Alpha=1
    GravityOverride:Vector="0,0,0"
    ComponentLinearAccScale:Vector="0.3,0.3,0.3"
    ComponentLinearVelScale:Vector="0.3,0.3,0.3"
    ExternalForce="0,0,0"
   ) -> (Pose) {
    angularDampingOverride=0.9999
    angularSpringConstant=2500
    bAngularSpring=true
    bOverrideAngularDamping=true
    bUseGravityOverride=true
    gravityScale=0.35
    simulationSpace=RootRelative
    boundBone {boneName=dyn_faceframe02_rt}
    componentAppliedLinearAccClamp=<3000 3000 3000>
    componentLinearAccScale=<0.3 0.3 0.3>
    componentLinearVelScale=<0.3 0.3 0.3>
    gravityOverride {z=-980}
    physicsBodyDefinitions [{
      boundBone {boneName=dyn_faceframe02_rt}
      boxExtents=<27 27 27>
      constraintSetup {angularLimitsMax=<0 8 5>, angularLimitsMin=<0 -8 -5>, angularTarget=<1 0 0>}
      localJointOffset=<3.2 0 0>
     }]
    simSpaceSettings {maxAngularAcceleration=50, maxAngularVelocity=100, simSpaceAngularAlpha=0.2}
   }
   n6 = Anim Dynamics - Bone: dyn_faceframe01_lf!(
    ComponentPose=@n5
    Alpha=1
    GravityOverride:Vector="0,0,0"
    ComponentLinearAccScale:Vector="0.3,0.3,0.3"
    ComponentLinearVelScale:Vector="0.3,0.3,0.3"
    ExternalForce="0,0,0"
   ) -> (Pose) {
    angularDampingOverride=0.9999
    angularSpringConstant=2500
    bAngularSpring=true
    bOverrideAngularDamping=true
    bUseGravityOverride=true
    gravityScale=0.35
    simulationSpace=RootRelative
    boundBone {boneName=dyn_faceframe01_lf}
    componentAppliedLinearAccClamp=<3000 3000 3000>
    componentLinearAccScale=<0.3 0.3 0.3>
    componentLinearVelScale=<0.3 0.3 0.3>
    gravityOverride {z=-980}
    physicsBodyDefinitions [{
      boundBone {boneName=dyn_faceframe01_lf}
      boxExtents=<27 27 27>
      constraintSetup {angularLimitsMax=<0 8 5>, angularLimitsMin=<0 -8 -5>, angularTarget=<1 0 0>}
      localJointOffset=<3.2 0 0>
     }]
    simSpaceSettings {maxAngularAcceleration=50, maxAngularVelocity=100, simSpaceAngularAlpha=0.2}
   }
   n7 = Anim Dynamics - Bone: dyn_faceframe02_lf!(
    ComponentPose=@n6
    Alpha=1
    GravityOverride:Vector="0,0,0"
    ComponentLinearAccScale:Vector="0.3,0.3,0.3"
    ComponentLinearVelScale:Vector="0.3,0.3,0.3"
    ExternalForce="0,0,0"
   ) -> (Pose) {
    angularDampingOverride=0.9999
    angularSpringConstant=2500
    bAngularSpring=true
    bOverrideAngularDamping=true
    bUseGravityOverride=true
    gravityScale=0.35
    simulationSpace=RootRelative
    boundBone {boneName=dyn_faceframe02_lf}
    componentAppliedLinearAccClamp=<3000 3000 3000>
    componentLinearAccScale=<0.3 0.3 0.3>
    componentLinearVelScale=<0.3 0.3 0.3>
    gravityOverride {z=-980}
    physicsBodyDefinitions [{
      boundBone {boneName=dyn_faceframe02_lf}
      boxExtents=<28 28 28>
      constraintSetup {angularLimitsMax=<0 8 5>, angularLimitsMin=<0 -5 -5>, angularTarget=<1 0 0>}
      localJointOffset=<3.2 0 0>
     }]
    simSpaceSettings {maxAngularAcceleration=50, maxAngularVelocity=100, simSpaceAngularAlpha=0.2}
   }
   n8 = Anim Dynamics - Bone: dyn_hairMain_rt!(
    ComponentPose=@n7
    Alpha=1
    GravityOverride:Vector="0,0,0"
    ComponentLinearAccScale:Vector="0.3,0.3,0.3"
    ComponentLinearVelScale:Vector="0.3,0.3,0.3"
    ExternalForce="0,0,0"
   ) -> (Pose) {
    angularDampingOverride=0.9999
    angularSpringConstant=2500
    bAngularSpring=true
    bOverrideAngularDamping=true
    bUseGravityOverride=true
    gravityScale=0.35
    simulationSpace=RootRelative
    boundBone {boneName=dyn_hairMain_rt}
    componentAppliedLinearAccClamp=<3000 3000 3000>
    componentLinearAccScale=<0.3 0.3 0.3>
    componentLinearVelScale=<0.3 0.3 0.3>
    gravityOverride {z=-980}
    physicsBodyDefinitions [{
      boundBone {boneName=dyn_hairMain_rt}
      boxExtents=<23 23 23>
      constraintSetup {angularLimitsMax=<8 8 10>, angularLimitsMin=<-5 -8 -10>, angularTarget=<1 0 0>}
      localJointOffset=<1.5 0 0>
     }]
    simSpaceSettings {maxAngularAcceleration=30, maxAngularVelocity=20, simSpaceAngularAlpha=0.2}
   }
   n9 = Anim Dynamics - Bone: dyn_hairMain_cr!(
    ComponentPose=@n8
    Alpha=1
    GravityOverride:Vector="0,0,0"
    ComponentLinearAccScale:Vector="0.3,0.3,0.3"
    ComponentLinearVelScale:Vector="0.3,0.3,0.3"
    ExternalForce="0,0,0"
   ) -> (Pose) {
    angularDampingOverride=0.9999
    angularSpringConstant=2500
    bAngularSpring=true
    bOverrideAngularDamping=true
    bUseGravityOverride=true
    gravityScale=0.35
    simulationSpace=RootRelative
    boundBone {boneName=dyn_hairMain_cr}
    componentAppliedLinearAccClamp=<3000 3000 3000>
    componentLinearAccScale=<0.3 0.3 0.3>
    componentLinearVelScale=<0.3 0.3 0.3>
    gravityOverride {z=-980}
    physicsBodyDefinitions [{
      boundBone {boneName=dyn_hairMain_cr}
      boxExtents=<17 17 17>
      constraintSetup {angularLimitsMax=<8 13 10>, angularLimitsMin=<-5 -13 -10>, angularTarget=<1 0 0>}
      localJointOffset=<1.5 0 0>
     }]
    simSpaceSettings {maxAngularAcceleration=30, maxAngularVelocity=20, simSpaceAngularAlpha=0.2}
   }
   n10 = Anim Dynamics - Bone: dyn_hairMain_lf!(
    ComponentPose=@n9
    Alpha=1
    GravityOverride:Vector="0,0,0"
    ComponentLinearAccScale:Vector="0.3,0.3,0.3"
    ComponentLinearVelScale:Vector="0.3,0.3,0.3"
    ExternalForce="0,0,0"
   ) -> (Pose) {
    angularDampingOverride=0.9999
    angularSpringConstant=2500
    bAngularSpring=true
    bOverrideAngularDamping=true
    bUseGravityOverride=true
    gravityScale=0.35
    simulationSpace=RootRelative
    boundBone {boneName=dyn_hairMain_lf}
    componentAppliedLinearAccClamp=<3000 3000 3000>
    componentLinearAccScale=<0.3 0.3 0.3>
    componentLinearVelScale=<0.3 0.3 0.3>
    gravityOverride {z=-980}
    physicsBodyDefinitions [{
      boundBone {boneName=dyn_hairMain_lf}
      boxExtents=<35 35 35>
      constraintSetup {angularLimitsMax=<8 5 7>, angularLimitsMin=<-5 -5 -7>, angularTarget=<1 0 0>}
      localJointOffset=<1.5 0 0>
     }]
    simSpaceSettings {maxAngularAcceleration=30, maxAngularVelocity=20, simSpaceAngularAlpha=0.2}
   }
   n11 = Anim Dynamics - Bone: dyn_bangs_cr!(
    ComponentPose=@n10
    Alpha=1
    GravityOverride:Vector="0,0,0"
    ComponentLinearAccScale:Vector="0.3,0.3,0.3"
    ComponentLinearVelScale:Vector="0.3,0.3,0.3"
    ExternalForce="0,0,0"
   ) -> (Pose) {
    angularDampingOverride=0.9999
    angularSpringConstant=2500
    bAngularSpring=true
    bOverrideAngularDamping=true
    bUseGravityOverride=true
    gravityScale=0.35
    simulationSpace=RootRelative
    boundBone {boneName=dyn_bangs_cr}
    componentAppliedLinearAccClamp=<3000 3000 3000>
    componentLinearAccScale=<0.3 0.3 0.3>
    componentLinearVelScale=<0.3 0.3 0.3>
    gravityOverride {z=-980}
    physicsBodyDefinitions [{
      boundBone {boneName=dyn_bangs_cr}
      boxExtents=<25 25 25>
      constraintSetup {angularLimitsMax=<0 10 15>, angularLimitsMin=<0 -10 -8>, angularTarget=<1 0 0>}
      localJointOffset=<3 0 0>
     }]
    simSpaceSettings {maxAngularAcceleration=50, maxAngularVelocity=100, simSpaceAngularAlpha=0.2}
   }
   n12 = Anim Dynamics - Bone: dyn_bangs01_lf!(
    ComponentPose=@n11
    Alpha=1
    GravityOverride:Vector="0,0,0"
    ComponentLinearAccScale:Vector="0.3,0.3,0.3"
    ComponentLinearVelScale:Vector="0.3,0.3,0.3"
    ExternalForce="0,0,0"
   ) -> (Pose) {
    angularDampingOverride=0.9999
    angularSpringConstant=2500
    bAngularSpring=true
    bOverrideAngularDamping=true
    bUseGravityOverride=true
    gravityScale=0.35
    simulationSpace=RootRelative
    boundBone {boneName=dyn_bangs01_lf}
    componentAppliedLinearAccClamp=<3000 3000 3000>
    componentLinearAccScale=<0.3 0.3 0.3>
    componentLinearVelScale=<0.3 0.3 0.3>
    gravityOverride {z=-980}
    physicsBodyDefinitions [{
      boundBone {boneName=dyn_bangs01_lf}
      boxExtents=<24 24 24>
      constraintSetup {angularLimitsMax=<0 11 12>, angularLimitsMin=<0 -8 -12>, angularTarget=<1 0 0>}
      localJointOffset=<3 0 0>
     }]
    simSpaceSettings {maxAngularAcceleration=50, maxAngularVelocity=100, simSpaceAngularAlpha=0.2}
   }
   n13 = Anim Dynamics - Bone: dyn_bangs02_lf!(
    ComponentPose=@n12
    Alpha=1
    GravityOverride:Vector="0,0,0"
    ComponentLinearAccScale:Vector="0.3,0.3,0.3"
    ComponentLinearVelScale:Vector="0.3,0.3,0.3"
    ExternalForce="0,0,0"
   ) -> (Pose) {
    angularDampingOverride=0.9999
    angularSpringConstant=2500
    bAngularSpring=true
    bOverrideAngularDamping=true
    bUseGravityOverride=true
    gravityScale=0.35
    simulationSpace=RootRelative
    boundBone {boneName=dyn_bangs02_lf}
    componentAppliedLinearAccClamp=<3000 3000 3000>
    componentLinearAccScale=<0.3 0.3 0.3>
    componentLinearVelScale=<0.3 0.3 0.3>
    gravityOverride {z=-980}
    physicsBodyDefinitions [{
      boundBone {boneName=dyn_bangs02_lf}
      boxExtents=<20 20 20>
      constraintSetup {angularLimitsMax=<0 6 2>, angularLimitsMin=<0 -3 -12>, angularTarget=<1 0 0>}
      localJointOffset=<3 0 0>
     }]
    simSpaceSettings {maxAngularAcceleration=50, maxAngularVelocity=100, simSpaceAngularAlpha=0.2}
   }
   n14 = Anim Dynamics - Bone: dyn_hairTop01_rt!(
    ComponentPose=@n13
    Alpha=1
    GravityOverride:Vector="0,0,0"
    ComponentLinearAccScale:Vector="0.4,0.4,0.4"
    ComponentLinearVelScale:Vector="0.4,0.4,0.4"
    ExternalForce="0,0,0"
   ) -> (Pose) {
    angularDampingOverride=0.9999
    angularSpringConstant=5000
    bAngularSpring=true
    bOverrideAngularDamping=true
    bUseGravityOverride=true
    gravityScale=0.35
    simulationSpace=RootRelative
    boundBone {boneName=dyn_hairTop01_rt}
    componentAppliedLinearAccClamp=<3000 3000 3000>
    componentLinearAccScale=<0.4 0.4 0.4>
    componentLinearVelScale=<0.4 0.4 0.4>
    gravityOverride {z=-980}
    physicsBodyDefinitions [{
      boundBone {boneName=dyn_hairTop01_rt}
      boxExtents=<20 20 20>
      constraintSetup {angularLimitsMax=<10 20 20>, angularLimitsMin=<-10 -10 -20>, angularTarget=<1 0 0>}
      localJointOffset=<1.5 0 0>
     }]
    simSpaceSettings {maxAngularAcceleration=30, maxAngularVelocity=20, simSpaceAngularAlpha=0.2}
   }
   n15 = Anim Dynamics - Bone: dyn_hairTop02_rt!(
    ComponentPose=@n14
    Alpha=1
    GravityOverride:Vector="0,0,0"
    ComponentLinearAccScale:Vector="0.4,0.4,0.4"
    ComponentLinearVelScale:Vector="0.4,0.4,0.4"
    ExternalForce="0,0,0"
   ) -> (Pose) {
    angularDampingOverride=0.9999
    angularSpringConstant=5000
    bAngularSpring=true
    bOverrideAngularDamping=true
    bUseGravityOverride=true
    gravityScale=0.35
    simulationSpace=RootRelative
    boundBone {boneName=dyn_hairTop02_rt}
    componentAppliedLinearAccClamp=<3000 3000 3000>
    componentLinearAccScale=<0.4 0.4 0.4>
    componentLinearVelScale=<0.4 0.4 0.4>
    gravityOverride {z=-980}
    physicsBodyDefinitions [{
      boundBone {boneName=dyn_hairTop02_rt}
      boxExtents=<25 25 25>
      constraintSetup {angularLimitsMax=<8 15 20>, angularLimitsMin=<-8 -10 -20>, angularTarget=<1 0 0>}
      localJointOffset=<1.5 0 0>
     }]
    simSpaceSettings {maxAngularAcceleration=30, maxAngularVelocity=20, simSpaceAngularAlpha=0.2}
   }
   n16 = Anim Dynamics - Bone: dyn_hairTop_lf!(
    ComponentPose=@n15
    Alpha=1
    GravityOverride:Vector="0,0,0"
    ComponentLinearAccScale:Vector="0.4,0.4,0.4"
    ComponentLinearVelScale:Vector="0.4,0.4,0.4"
    ExternalForce="0,0,0"
   ) -> (Pose) {
    angularDampingOverride=0.9999
    angularSpringConstant=5000
    bAngularSpring=true
    bOverrideAngularDamping=true
    bUseGravityOverride=true
    gravityScale=0.35
    simulationSpace=RootRelative
    boundBone {boneName=dyn_hairTop_lf}
    componentAppliedLinearAccClamp=<3000 3000 3000>
    componentLinearAccScale=<0.4 0.4 0.4>
    componentLinearVelScale=<0.4 0.4 0.4>
    gravityOverride {z=-980}
    physicsBodyDefinitions [{
      boundBone {boneName=dyn_hairTop_lf}
      boxExtents=<18 18 18>
      constraintSetup {angularLimitsMax=<10 15 15>, angularLimitsMin=<-10 -10 -15>, angularTarget=<1 0 0>}
      localJointOffset=<1.5 0 0>
     }]
    simSpaceSettings {maxAngularAcceleration=30, maxAngularVelocity=20, simSpaceAngularAlpha=0.2}
   }
   n17 = Anim Dynamics - Bone: dyn_hairTop_cr!(
    ComponentPose=@n16
    Alpha=1
    GravityOverride:Vector="0,0,0"
    ComponentLinearAccScale:Vector="0.4,0.4,0.4"
    ComponentLinearVelScale:Vector="0.4,0.4,0.4"
    ExternalForce="0,0,0"
   ) -> (Pose) {
    angularDampingOverride=0.9999
    angularSpringConstant=2500
    bAngularSpring=true
    bOverrideAngularDamping=true
    bUseGravityOverride=true
    gravityScale=0.35
    simulationSpace=RootRelative
    boundBone {boneName=dyn_hairTop_cr}
    componentAppliedLinearAccClamp=<3000 3000 3000>
    componentLinearAccScale=<0.4 0.4 0.4>
    componentLinearVelScale=<0.4 0.4 0.4>
    gravityOverride {z=-980}
    physicsBodyDefinitions [{
      boundBone {boneName=dyn_hairTop_cr}
      boxExtents=<25 25 25>
      constraintSetup {angularLimitsMax=<0 5 10>, angularLimitsMin=<0 -5 -10>, angularTarget=<1 0 0>}
      localJointOffset=<2.5 0 0>
     }]
    simSpaceSettings {maxAngularAcceleration=50, maxAngularVelocity=100, simSpaceAngularAlpha=0.2}
   }
   n18 = Anim Dynamics - Bone: dyn_hairCowlick01_cr!(
    ComponentPose=@n17
    Alpha=1
    GravityOverride:Vector="0,0,0"
    ComponentLinearAccScale:Vector="0.4,0.4,0.4"
    ComponentLinearVelScale:Vector="0.4,0.4,0.4"
    ExternalForce="0,0,0"
   ) -> (Pose) {
    angularDampingOverride=0.9999
    angularSpringConstant=2500
    bAngularSpring=true
    bOverrideAngularDamping=true
    bUseGravityOverride=true
    gravityScale=0.35
    simulationSpace=RootRelative
    boundBone {boneName=dyn_hairCowlick01_cr}
    componentAppliedLinearAccClamp=<3000 3000 3000>
    componentLinearAccScale=<0.4 0.4 0.4>
    componentLinearVelScale=<0.4 0.4 0.4>
    gravityOverride {z=-980}
    physicsBodyDefinitions [{
      boundBone {boneName=dyn_hairCowlick01_cr}
      boxExtents=<28 28 28>
      constraintSetup {angularLimitsMax=<0 10 10>, angularLimitsMin=<0 -5 -10>, angularTarget=<1 0 0>}
      localJointOffset=<2 0 0>
     }]
    simSpaceSettings {maxAngularAcceleration=50, maxAngularVelocity=100, simSpaceAngularAlpha=0.2}
   }
   n19 = Anim Dynamics - Bone: dyn_hairCowlick02_cr!(
    ComponentPose=@n18
    Alpha=1
    GravityOverride:Vector="0,0,0"
    ComponentLinearAccScale:Vector="0.4,0.4,0.4"
    ComponentLinearVelScale:Vector="0.4,0.4,0.4"
    ExternalForce="0,0,0"
   ) -> (Pose) {
    angularDampingOverride=0.9999
    angularSpringConstant=2500
    bAngularSpring=true
    bOverrideAngularDamping=true
    bUseGravityOverride=true
    gravityScale=0.35
    simulationSpace=RootRelative
    boundBone {boneName=dyn_hairCowlick02_cr}
    componentAppliedLinearAccClamp=<3000 3000 3000>
    componentLinearAccScale=<0.4 0.4 0.4>
    componentLinearVelScale=<0.4 0.4 0.4>
    gravityOverride {z=-980}
    physicsBodyDefinitions [{
      boundBone {boneName=dyn_hairCowlick02_cr}
      boxExtents=<28 28 28>
      constraintSetup {angularLimitsMax=<0 10 10>, angularLimitsMin=<0 -5 -10>, angularTarget=<1 0 0>}
      localJointOffset=<2 0 0>
     }]
    simSpaceSettings {maxAngularAcceleration=50, maxAngularVelocity=100, simSpaceAngularAlpha=0.2}
   }
   Output Pose(Result=Component To Local(ComponentPose=@n19)) {layerGroup=None}
  ]}
 EventGraph {nodes [
   Event BlueprintUpdateAnimation()
   > n1 = "Gravity Override State Function (Parameter Asset) (Nativized)"(
    CharacterPartAnimInstance=Self-Reference()
    GameTimeInSeconds=GetGameTimeInSeconds()
    Parameters=&/M_MED_ZenCrash_LandFoil_FaceAcc_Parameters
   )
   > Set Gravity Override(Gravity Override=@n1)
   > Set RBAN Settings Out(RBAN Settings Out=@n1.OutRigidBodyAnimNodeParameters)
   n2 = Event BlueprintProcessAnimInputEvent()
   n3 = IsNamed(self=@n2.AnimInputEvent, InEventName=aie.Player.Zipline.Entered)
   n4 = OR Boolean(A=Get bIsMoving2D(), B=Get bIsTacticalSprinting())
   n5 = Get bIsJumping()
   > n6 = Branch(Condition=OR Boolean(A=@n3, B=AND Boolean!(A=@n3, B=@n5, C:bool=@n4) -> (ReturnValue)))
   > Set animDyn on zipline_03(animDyn on zipline_03:Vector="0, 0, 0") -> (Output_Get:Vector)
   > Set animDyn on zipline_04(animDyn on zipline_04:Vector="0, 0, 0") -> (Output_Get:Vector)
   Branch(execute=@n6.else, Condition=IsNamed(self=@n2.AnimInputEvent, InEventName=aie.Player.Zipline.Exited))
   > Set animDyn on zipline_03(animDyn on zipline_03="0.3,0.3,0.3")
   > Set animDyn on zipline_04(animDyn on zipline_04="0.4,0.4,0.4")
  ]}
]
variables [
 animDyn on zipline_03 {type=Vector, default=<0.3 0.3 0.3>}
 animDyn on zipline_04 {type=Vector, default=<0.4 0.4 0.4>}
 Gravity Override {type=Vector}
 RBAN Settings Out {type=FortRigidBodyAnimNodeParameters}
]
